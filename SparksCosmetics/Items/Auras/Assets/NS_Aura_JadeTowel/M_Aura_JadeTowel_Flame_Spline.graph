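# Unlit, two-sided translucent flame material for Niagara ribbons and mesh particles in the Jade Towel aura.
# Two panning noises distort a smoke-trail texture, which drives a three-color emissive ramp with length-wise tint and particle color.
# Opacity combines that texture with noise-warped edge masks, depth fade, optional angle fade and base-to-tip power.
# World position offset layers SplineThicken, sine/cosine wobble, clamped Niagara-driven pull and optional camera offset; Rotate UVs switches the flow axis.
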Material /SparksCosmetics/Items/Auras/Assets/NS_Aura_JadeTowel/M_Aura_JadeTowel_Flame_Spline
graph {
 properties {
  bUsedWithNiagaraMeshParticles=true
  bUsedWithNiagaraRibbons=true
  blendMode=BLEND_Translucent
  pixelDepthOffsetMode=PDOM_Legacy
  shadingModel=MSM_Unlit
  twoSided=true
  usedShadingModels=Unlit
 }
 nodes [
  n1 = MaterialFunctionCall(Float2=TextureCoordinate()) {materialFunction=&BreakOutFloat2Components}
  n2 = OneMinus(None=@n1.R)
  n3 = Add(A=Time(), B=ParticleRandom())
  n4 = LinearInterpolate(Alpha=ParticleRandom()) {constA=0.7, constB=1.3}
  n5 = Multiply(
   A=Add(A=StaticSwitchParameter(True=@n1.G, False=@n2) {defaultValue=true, parameterName=Rotate UVs}, B=Multiply(A=@n3, B=@n4))
   B=ScalarParameter() {defaultValue=0.5, group=WPO, parameterName=WPOSineAmount, sortPriority=10}
  )
  n6 = Multiply(A=Sine(None=@n5), B=Constant3Vector() {constant=<rgba 0 0.75 0 1>})
  n7 = Multiply(A=Cosine(None=@n5), B=Constant3Vector() {constant=<rgba 0.75 0 0 1>})
  n8 = MaterialFunctionCall(Float2=TextureCoordinate()) {materialFunction=&BreakOutFloat2Components}
  n9 = Multiply(
   A=Multiply(A=Add(A=@n6, B=@n7), B=ScalarParameter() {defaultValue=15, group=WPO, parameterName=WPOSineIntensity, sortPriority=11})
   B=StaticSwitchParameter(True=OneMinus(None=@n8.G), False=@n8.R) {defaultValue=true, parameterName=Rotate UVs}
  )
  n10 = DynamicParameter() {desc=Parameters set by Niagara, parameterIndex=3, defaultValue=<rgba 0 1 1 1>, paramNames [X,Y,Z,Intensity]}
  n11 = MaterialFunctionCall(Float3=Subtract(
    A=MaterialFunctionCall(X=@n10.X, Y=@n10.X, Z=@n10.X) {materialFunction=&MakeFloat3}.Result
    B=ParticlePositionWS()
   )) {materialFunction=&BreakOutFloat3Components}
  n12 = ScalarParameter() {defaultValue=100, desc=Interpolate Clamp XY, group=WPO, parameterName=Interpolate XY Clamp, sortPriority=3}
  n13 = ScalarParameter() {defaultValue=2, desc=Interpolate Clamp Z, group=WPO, parameterName=Interpolate Clamp Z, sortPriority=4}
  n14 = AppendVector(
   A=Clamp(None=AppendVector(A=@n11.R, B=@n11.R), Min=Multiply(A=@n12) {constB=-1}, Max=@n12) {maxDefault=1000, minDefault=-1000}
   B=Clamp(None=@n11.B, Min=Multiply(A=@n13) {constB=-1}, Max=@n13) {minDefault=-1}
  )
  n15 = MaterialFunctionCall(Float2=TextureCoordinate()) {materialFunction=&BreakOutFloat2Components}
  n16 = OneMinus(None=@n15.G)
  n17 = Multiply(
   A=@n10.Intensity
   B=Power(
    Base=StaticSwitchParameter(True=@n16, False=@n15.R) {defaultValue=true, parameterName=Rotate UVs}
    Exp=ScalarParameter() {defaultValue=5, desc=Interpolate Power, group=WPO, parameterName=Interpolate Power, sortPriority=2}
   )
  )
  n18 = DynamicParameter() {defaultValue=<rgba 1 25 25 0>, paramNames [Random Offset, "Spline Width (Base)", "Spline Width (Tip)", Camera Offset]}
  n19 = Add(
   A=StaticSwitchParameter(True=LinearInterpolate(B=@n14, Alpha=@n17), False=Constant()) {defaultValue=true, group=WPO, parameterName=Use Interpolate WPO, sortPriority=1}
   B=MaterialFunctionCall(WidthBase=@n18."Spline Width (Base)", WidthTip=@n18."Spline Width (Base)") {materialFunction=&SplineThicken}.WorldPositionOffset
  )
  n20 = DynamicParameter() {defaultValue=<rgba 1 25 25 0>, paramNames [Random Offset, "Spline Width (Base)", "Spline Width (Tip)", Camera Offset]}
  Add(
   A=StaticSwitchParameter(True=Add(A=@n9, B=@n19), False=@n19) {defaultValue=true, group=WPO, parameterName=Use Wobble WPO, sortPriority=9}
   B=LinearInterpolate(
    B=MaterialFunctionCall(Offset Amount=@n20.Camera Offset) {materialFunction=/Engine/Functions/Engine_MaterialFunctions02/WorldPositionOffset/CameraOffset}.Camera Offset
    Alpha=ScalarParameter() {group=WPO, parameterName=Use Camera Offset}
   )
  )
  n21 = VectorParameter() {group=Color, parameterName=Bottom Color, sortPriority=13, defaultValue=<rgba 2 2 2 1>}
  n22 = VectorParameter() {group=Color, parameterName=Top Color, sortPriority=11, defaultValue=<rgba 0 0 1 1>}
  n23 = MaterialFunctionCall(Float2=TextureCoordinate()) {materialFunction=&BreakOutFloat2Components}
  n24 = StaticSwitchParameter(True=OneMinus(None=@n23.G), False=@n23.R) {defaultValue=true, parameterName=Rotate UVs}
  n25 = VectorParameter() {group=Color, parameterName=Color Mid, sortPriority=3, defaultValue=<rgba 0 1 0.7438 0.5>}
  n26 = TextureCoordinate()
  n27 = MaterialFunctionCall(Float2=@n26) {materialFunction=&BreakOutFloat2Components}
  n28 = VectorParameter() {
   group=MainTexture
   parameterName=MainTexture_UVControl
   channelNames {a=VPan, b=UPan, g=VScale, r=UScale}
   defaultValue=<rgba 0.5 1.2 -0.35 0>
  }
  n29 = AppendVector(A=@n28.UScale, B=@n28.UScale)
  n30 = Multiply(A=StaticSwitchParameter(True=AppendVector(A=@n27.G, B=@n27.G), False=@n26) {defaultValue=true, parameterName=Rotate UVs}, B=@n29)
  n31 = Constant() {r=0.5}
  n32 = Constant() {r=-0.5}
  n33 = AppendVector(A=@n28.UPan, B=@n28.UPan)
  n34 = Time() {bOverride_Period=true, period=13333}
  n35 = DynamicParameter() {defaultValue=<rgba 1 25 25 0>, paramNames [Random Offset, "Spline Width (Base)", "Spline Width (Tip)", Camera Offset]}
  n36 = MaterialFunctionCall(Float2=TextureCoordinate()) {materialFunction=&BreakOutFloat2Components}
  n37 = Power(
   Base=StaticSwitchParameter(True=@n36.G, False=OneMinus(None=@n36.R)) {defaultValue=true, parameterName=Rotate UVs}
   Exp=ScalarParameter() {defaultValue=5, group=NoiseB, parameterName=NoiseB_UVGradient}
  )
  n38 = TextureCoordinate()
  n39 = MaterialFunctionCall(Float2=@n38) {materialFunction=&BreakOutFloat2Components}
  n40 = StaticSwitchParameter(True=AppendVector(A=@n39.G, B=@n39.G), False=@n38) {defaultValue=true, parameterName=Rotate UVs}
  n41 = VectorParameter() {group=NoiseB, parameterName=NoiseB_UVControl, channelNames {a=VPan, b=UPan, g=VScale, r=UScale}, defaultValue=<rgba 0.5 0.5 0.5 -0.1>}
  n42 = AppendVector(A=@n41.UScale, B=@n41.UScale)
  n43 = DynamicParameter() {defaultValue=<rgba 1 25 25 0>, paramNames [Random Offset, "Spline Width (Base)", "Spline Width (Tip)", Camera Offset]}
  n44 = Add(
   A=Multiply(A=AppendVector(A=@n41.UPan, B=@n41.UPan), B=Time() {bOverride_Period=true, period=13333})
   B=AppendVector(A=Multiply(A=@n43.Random Offset) {constB=0.9}, B=Constant())
  )
  n45 = ScalarParameter() {group=MainTexture, parameterName=UseDynamicOffset}
  n46 = TextureSampleParameter2D(UVs=Add(A=Multiply(A=@n40, B=@n42), B=LinearInterpolate(B=@n44, Alpha=@n45))) {
   group=NoiseB
   parameterName=Noise B
   samplerType=SAMPLERTYPE_LinearGrayscale
   texture=&T_SimplePerlin_Greyscale
  }
  n47 = VectorParameter() {
   group=NoiseB
   parameterName=NoiseB_NoiseControl
   sortPriority=60
   channelNames {a=VStrength, b=UStrength, g=VOffset, r=UOffset}
   defaultValue=<rgba -0.5 -0.5 0 0.5>
  }
  n48 = Add(A=AppendVector(A=@n46.R, B=@n46.R), B=AppendVector(A=@n47.UOffset, B=@n47.UOffset))
  n49 = AppendVector(A=@n47.UStrength, B=@n47.UStrength)
  n50 = TextureCoordinate()
  n51 = MaterialFunctionCall(Float2=@n50) {materialFunction=&BreakOutFloat2Components}
  n52 = StaticSwitchParameter(True=AppendVector(A=@n51.G, B=@n51.G), False=@n50) {defaultValue=true, parameterName=Rotate UVs}
  n53 = VectorParameter() {group=NoiseA, parameterName=NoiseA_UVControl, channelNames {a=VPan, b=UPan, g=VScale, r=UScale}, defaultValue=<rgba 0.2 0.6 0.34 0.1>}
  n54 = AppendVector(A=@n53.UScale, B=@n53.UScale)
  n55 = DynamicParameter() {defaultValue=<rgba 1 25 25 0>, paramNames [Random Offset, "Spline Width (Base)", "Spline Width (Tip)", Camera Offset]}
  n56 = Add(
   A=Multiply(A=AppendVector(A=@n53.UPan, B=@n53.UPan), B=Time() {bOverride_Period=true, period=13333})
   B=AppendVector(A=Multiply(A=@n55.Random Offset) {constB=1.1}, B=Constant())
  )
  n57 = ScalarParameter() {group=MainTexture, parameterName=UseDynamicOffset}
  n58 = TextureSampleParameter2D(UVs=Add(A=Multiply(A=@n52, B=@n54), B=LinearInterpolate(B=@n56, Alpha=@n57))) {
   group=NoiseA
   parameterName=Noise A
   samplerType=SAMPLERTYPE_LinearGrayscale
   texture=/Game/Global/Textures/Default/Tileable_Patterns/T_StormCloudDetail01_M
  }
  n59 = AppendVector(A=@n58.R, B=@n58.R)
  n60 = VectorParameter() {
   group=NoiseA
   parameterName=NoiseA_NoiseControl
   sortPriority=60
   channelNames {a=VStrength, b=UStrength, g=VOffset, r=UOffset}
   defaultValue=<rgba -0.5 -0.604 0.2 0.6>
  }
  n61 = AppendVector(A=@n60.UOffset, B=@n60.UOffset)
  n62 = Add(
   A=Multiply(A=OneMinus(None=@n37), B=Multiply(A=@n48, B=@n49))
   B=Multiply(A=Add(A=@n59, B=@n61), B=AppendVector(A=@n60.UStrength, B=@n60.UStrength))
  )
  n63 = Add(A=@n62, B=AppendVector(A=@n35.Random Offset, B=Constant()))
  n64 = ScalarParameter() {group=MainTexture, parameterName=UseDynamicOffset}
  n65 = Add(A=Add(A=Add(A=@n30, B=@n31), B=Multiply(A=@n29, B=@n32)), B=Add(A=Multiply(A=@n33, B=@n34), B=LinearInterpolate(A=@n62, B=@n63, Alpha=@n64)))
  n66 = TextureSampleParameter2D(UVs=@n65) {group=MainTexture, parameterName=Main Texture, texture=/Game/Effects/Fort_Effects/Textures/Smoke/T_Smoke_Trail_Soft_2}
  n67 = Power(Base=@n66.G, Exp=ScalarParameter() {defaultValue=1, group=Color, parameterName=Crunch Colors}) {constExponent=3}
  n68 = Saturate(None=Multiply(A=@n67) {constB=2})
  n69 = Saturate(None=Add(A=Multiply(A=@n67) {constB=2}) {constB=-1})
  n70 = LinearInterpolate(
   A=LinearInterpolate(A=VectorParameter() {group=Color, parameterName=Color Low, sortPriority=5, defaultValue=<rgba 0 0.1113 1 0.25>}.RGB, B=@n25.RGB, Alpha=@n68)
   B=LinearInterpolate(A=@n25.RGB, B=VectorParameter() {group=Color, parameterName=Color High, sortPriority=1, defaultValue=<rgba 1 1 1 1>}.RGB, Alpha=@n69)
   Alpha=@n66.G
  ) {constA=1}
  n71 = ParticleColor()
  MaterialFunctionCall(Emissive=Multiply(A=Multiply(A=@n70, B=@n71.RGB), B=LinearInterpolate(A=@n21.RGB, B=@n22.RGB, Alpha=@n24))) {materialFunction=&MF_FX_EmissiveAdjustments}
  n72 = VectorParameter() {group=Color, parameterName=Color Mid, sortPriority=3, defaultValue=<rgba 0 1 0.7438 0.5>}
  n73 = LinearInterpolate(
   A=LinearInterpolate(A=VectorParameter() {group=Color, parameterName=Color Low, sortPriority=5, defaultValue=<rgba 0 0.1113 1 0.25>}.A, B=@n72.A, Alpha=@n68)
   B=LinearInterpolate(A=@n72.A, B=VectorParameter() {group=Color, parameterName=Color High, sortPriority=1, defaultValue=<rgba 1 1 1 1>}.A, Alpha=@n69)
   Alpha=@n66.G
  ) {constA=1}
  n74 = Multiply(
   A=Desaturation(None=@n66.RGB, Fraction=Constant() {r=1})
   B=MaterialFunctionCall!(TexCoord=@n65, ClampChannel=Constant2Vector() {g=1}) {materialFunction=/Game/Effects/Fort_Effects/MaterialFunctions/MF_FX_TilingMethod}.Result
  )
  n75 = ScalarParameter() {defaultValue=1, group=Opacity, parameterName=Contrast A}
  n76 = MaterialFunctionCall(Float2=TextureCoordinate()) {materialFunction=&BreakOutFloat2Components}
  n77 = StaticSwitchParameter(True=OneMinus(None=@n76.G), False=@n76.R) {defaultValue=true, parameterName=Rotate UVs}
  n78 = VectorParameter() {group=EdgeMask, parameterName=Edge Mask Top, channelNames {a="", b=Power, g=Multiply, r=Offset}, defaultValue=<rgba 0 3 2 1>}
  n79 = Constant4Vector() {constant=<rgba 1 1 0 0>}
  n80 = MaterialFunctionCall!(
   UorV=LinearInterpolate(A=@n77, B=Add(A=@n77, B=@n62), Alpha=ScalarParameter() {defaultValue=1, group=EdgeMask, parameterName=EdgeMaskUVNoiseStr})
   Invert_EdgeControl=@n78
   NoInvert_EdgeControl=@n79.RGBA
  ) {materialFunction=/Game/Effects/Niagara/MasterMaterials/Functions/Common/FXCommon_EdgeMask_Ex}
  n81 = MaterialFunctionCall(Float2=@n77) {materialFunction=&BreakOutFloat2Components}
  n82 = VectorParameter() {group=EdgeMask, parameterName=Edge Mask Bottom, channelNames {a="", b=Power, g=Multiply, r=Offset}, defaultValue=<rgba 0 5 2 1>}
  n83 = MaterialFunctionCall(UorV=@n81.R, Invert_EdgeControl=@n79.RGBA, NoInvert_EdgeControl=@n82) {materialFunction=&FXCommon_EdgeMask_Ex}
  n84 = VectorParameter() {group=EdgeMask, parameterName=Edge Mask Sides, channelNames {a="", b=Power, g=Multiply, r=Offset}, defaultValue=<rgba 0 10 2 1>}
  n85 = Multiply(
   A=MaterialFunctionCall(In=Power(Base=@n74, Exp=@n75), Contrast=ScalarParameter() {group=Opacity, parameterName=Contrast B}) {materialFunction=&CheapContrast}.Result
   B=Multiply(
    A=Multiply(A=@n80.Mask, B=@n83.Mask)
    B=MaterialFunctionCall!(UorV=@n81.G, EdgeControl=@n84) {materialFunction=/Game/Effects/Niagara/MasterMaterials/Functions/Common/FXCommon_EdgeMask_Simple}.Mask
   )
  )
  n86 = ScalarParameter() {group=Opacity, parameterName=Angle Fade Sensitivity}
  n87 = LinearInterpolate(
   A=@n85
   B=Multiply(
    A=@n85
    B=MaterialFunctionCall("FadeSensitivity (optional)"=@n86) {materialFunction=&FadeBasedOnViewAngle}.Result
   )
   Alpha=ScalarParameter() {group=Opacity, parameterName=Use Angle Fade, sortPriority=30}
  )
  n88 = ScalarParameter() {defaultValue=1, group=Opacity, parameterName=Depth Fade}
  n89 = MaterialFunctionCall(Float2=TextureCoordinate()) {materialFunction=&BreakOutFloat2Components}
  n90 = StaticSwitchParameter(True=OneMinus(None=@n89.G), False=@n89.R) {defaultValue=true, parameterName=Rotate UVs}
  n91 = ScalarParameter() {defaultValue=1, group=Opacity, parameterName=Power Transition, sortPriority=39}
  Power(
   Base=Multiply(A=Multiply(A=@n71.A, B=@n73), B=DepthFade(Opacity=@n87, FadeDistance=@n88))
   Exp=LinearInterpolate(
    A=ScalarParameter() {defaultValue=0.4, group=Opacity, parameterName=Base Power, sortPriority=37}
    B=ScalarParameter() {defaultValue=1.6, group=Opacity, parameterName=Tip Power, sortPriority=35}
    Alpha=Power(Base=@n90, Exp=@n91) {constExponent=1}
   ) {constA=1, constB=5}
  ) {constExponent=1}
 ]
}
