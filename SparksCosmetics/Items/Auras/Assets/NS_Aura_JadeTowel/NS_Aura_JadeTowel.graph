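# JadeTowel aura cosmetic Niagara system that wraps a character's skeletal mesh in teal-tinted glows, flames and sparks.
# Sprite glows, ribbon flames, flame meshes and GPU sparks follow selected bones, several blending between paired companion bones.
# Flame ribbons rise with curl noise and inherited velocity, with first-particle locking and high-speed control.
# User flags adjust for large bodies and front end, while LOD distance tunes sparks.
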NiagaraSystem /SparksCosmetics/Items/Auras/Assets/NS_Aura_JadeTowel/NS_Aura_JadeTowel
dependencies {
 usedDataInterfaces [
  &NiagaraDataInterfaceArrayInt32
  &NiagaraDataInterfaceArrayFloat
  &NiagaraDataInterfaceCurve
  &NiagaraDataInterfaceColorCurve
  &NiagaraDataInterfaceArrayFloat3
  &NiagaraDataInterfaceSkeletalMesh
 ]
 usedDynamicInputs [&SimulationPosition, &VectorFromFloat]
 usedModules [
  &self:SystemSpawnScript.SystemScriptSource.SystemScriptGraph.NiagaraNodeAssignment_0.SetVariables_#1
  &self:SystemSpawnScript.SystemScriptSource.SystemScriptGraph.NiagaraNodeAssignment_4.SetVariables_#2
  &SystemState
 ]
}
emitters {
 BodyGlow {
  emitterUpdateScript [
   &EmitterState()
   &SpawnBurst_Instantaneous(Spawn Count=10, Use Spawn Probability=false, UseLoopCountLimit=true)
  ]
  particleSpawnScript [
   &PartitionParticles(Number Of Partitions=10)
   &InitializeParticle(
    Color=&MakeCustomLinearColorFromBool(Bool=$User.IsLargeBody, True Color=<rgba 0.3536 2 1.8366 1>, False Color=<rgba 0.1768 1 0.9183 0.4>)
    Lifetime=&RandomRangeFloat(Minimum=3, Maximum=4)
    Sprite Rotation Mode=::NewEnumerator2
    Sprite Size Mode=::NewEnumerator3
    Uniform Sprite Size=&MakeCustomFloatFromBool(Bool=$User.IsLargeBody, True Float=30, False Float=20)
    UsePositionOffset=false
    Write Material Random=true
   ) {module=InitializeParticle001}
   /SetVariables(
    Particles.PickBone:NiagaraInt32=&SelectIntFromArray(
     Direct Array Index=$Particles.PartitionParticles.Partition
     Int Selection Array=DataInterface() {propertyValues {IntData [1,2,3,4,3,4,5,6,5,6]}}
    )
    Particles.SymmetricalLerp:NiagaraFloat=&Multiply_Float(
     A=/Niagara/DynamicInputs/Integer/V2/Round_Float(Float:NiagaraFloat=&Multiply_FloatByInt(Float=0.49, Integer=$Particles.PartitionParticles.Partition))
     B=0.25
    )
   ) {isSetParameters=true}
   /SetVariables(
    Particles.glowCameraOffset=-10
    Particles.glowColor:LinearColor=&MakeCustomLinearColorFromBool(
     Bool=$User.IsFrontEnd
     True Color=&LerpLinearColors(StartColor=<rgba 0.1768 1 0.9183 0.1>, EndColor=<rgba 0.1768 0.7119 1 0.05>, LerpFactor=$Particles.SymmetricalLerp)
     False Color=&LerpLinearColors(StartColor=<rgba 0.1768 1 0.9183 0.15>, EndColor=<rgba 0.1768 0.7119 1 0.05>, LerpFactor=$Particles.SymmetricalLerp)
    )
    Particles.glowScale:Vector2f=&Vector2DFromFloat(Value=&Multiply_Float(
      A=&SelectFloatFromArray(
       Direct Array Index=$Particles.PartitionParticles.Partition
       Float Selection Array=DataInterface() {propertyValues {FloatData [60,60,40,40,40,40,30,30,30,30]}}
      )
      B=&MakeCustomFloatFromBool(Bool=$User.IsLargeBody, True Float=1.3, False Float=1)
     ))
   ) {isSetParameters=true}
  ]
  particleUpdateScript [
   &ParticleState(Kill Particles When Lifetime Has Elapsed=false, Let Infinitely Lived Particles Die When Emitter Deactivates=true)
   &CameraOffset(Camera Offset Amount=5)
   &ScaleSpriteSize(
    Scale Sprite Size Mode=::NewEnumerator0
    Uniform Scale Factor=&Waveform(
     Global Phase=&Add_Float(A=$Emitter.Age, B=&RandomRangeFloat(Minimum=0, Maximum=5))
     "Waveform [1]"=::NewEnumerator0
     "[1] Frequency"=1
     "[1] Amplitude Min/Max"=<0.85 1>
    )
   ) {enabled=false}
   &SpriteRotationRate(Rotation Rate=&RandomRangeFloat(Minimum=10, Maximum=20))
   &SkeletalMeshLocation(
    Bone / Socket Index=$Particles.PickBone
    Bone Sampling Mode=::NewEnumerator4
    MaskSpawnGroup=false
    Orientation Sampling=::NewEnumerator1
    Position Sampling=::NewEnumerator2
    Sampled Position Offset Coordinate Space:ENiagaraSkelMeshTransforms=::NewEnumerator1
    Skeletal Mesh=$System.Skeletal Mesh
   )
   &SkeletalMeshLocation(
    Bone / Socket Index=&SelectIntFromArray(Direct Array Index=$Particles.PickBone, Int Selection Array=$System.CompanionBone)
    Bone Sampling Mode=::NewEnumerator4
    MaskSpawnGroup=false
    Orientation Sampling=::NewEnumerator1
    Position Sampling=::NewEnumerator2
    Sampled Position Offset Coordinate Space:ENiagaraSkelMeshTransforms=::NewEnumerator1
    Skeletal Mesh=$System.Skeletal Mesh
   ) {module=SkeletalMeshLocation001}
   &DynamicMaterialParameters(
    Index 0 Param 1=&SelectFloatFromArray(
     Direct Array Index=$Particles.PartitionParticles.Partition
     Float Selection Array=DataInterface() {propertyValues {FloatData [5,5,5,4.5,4.5,4,4,3,3]}}
    )
    Index 0 Param 3=1
    Module.Param0WriteEnabled=true
    Module.Param1WriteEnabled=true
    Module.Param2WriteEnabled=true
    Module.Param3WriteEnabled=true
   )
   /SetVariables(Particles.Position=&LerpPosition(
     A=$Particles.SkeletalMeshLocation.SampledPosition
     B=$Particles.SkeletalMeshLocation001.SampledPosition
     Alpha=&SelectFloatFromArray(
      Direct Array Index=$Particles.PartitionParticles.Partition
      Float Selection Array=DataInterface() {propertyValues {FloatData [0,0,0.5,0.5,0,0,0.5,0.5,0,0]}}
     )
    )) {isSetParameters=true}
  ]
  rendererClasses [&NiagaraSpriteRendererProperties]
  renderers [
   {rendererClass=&NiagaraSpriteRendererProperties}
   {rendererClass=&NiagaraSpriteRendererProperties, rendererIndex=1}
   {rendererClass=&NiagaraSpriteRendererProperties, rendererIndex=2}
  ]
 }
 FN_Debug {
  bEnabled=false
  emitterUpdateScript [
   &EmitterState()
   &SpawnBurst_Instantaneous(Use Spawn Probability=false, UseLoopCountLimit=true)
  ]
  particleSpawnScript [&InitializeParticle(
    Lifetime=5
    Mass Mode=::NewEnumerator1
    Position Offset=<40 0 160>
    Sprite Size=<75 75>
    Sprite Size Mode=::NewEnumerator4
    UsePositionOffset=true
    Write Material Random=true
   ) {module=InitializeParticle001}]
  particleUpdateScript [
   &ParticleState(Kill Particles When Lifetime Has Elapsed=false)
   null() {enabled=false, module=CountBeat}
   /SetVariables(Particles.Debug:NiagaraFloat=&MakeFloatFromInt(Int=&MakeIntFromFloat(Float=&Multiply_Float(A=$Engine.Owner.LODDistance, B=0.01)))) {isSetParameters=true}
   &DynamicMaterialParameters(
    Index 0 Param 1=$Particles.Debug
    Module.Index0Param0Write=true
    Module.Index0Param1Write=true
    Module.Index0Param2Write=true
    Module.Index0Param3Write=true
   )
  ]
  rendererClasses [&NiagaraSpriteRendererProperties]
  renderers [{rendererClass=&NiagaraSpriteRendererProperties}]
 }
 Flames {
  emitterUpdateScript [
   &EmitterState()
   &SpawnRate(SpawnRate=$System.RibbonSpawnRate, Use Spawn Probability=false) {enabled=false}
   &SpawnRate(SpawnGroup=1, SpawnRate=$System.RibbonSpawnRate, Use Spawn Probability=false) {module=SpawnRate001}
   &SpawnRate(SpawnGroup=2, SpawnRate=$System.RibbonSpawnRate, Use Spawn Probability=false) {module=SpawnRate002}
   &SpawnRate(SpawnGroup=3, SpawnRate=$System.RibbonSpawnRate, Use Spawn Probability=false) {module=SpawnRate003}
   &SpawnRate(SpawnGroup=4, SpawnRate=$System.RibbonSpawnRate, Use Spawn Probability=false) {module=SpawnRate004}
   &SpawnRate(SpawnGroup=5, SpawnRate=$System.RibbonSpawnRate, Use Spawn Probability=false) {module=SpawnRate005}
   &SpawnRate(SpawnGroup=6, SpawnRate=$System.RibbonSpawnRate, Use Spawn Probability=false) {module=SpawnRate006}
   &self:SetHasSpawned() {module=SetHasSpawned}
  ]
  particleSpawnScript [
   &InitializeParticle(
    Color=&MakeCustomLinearColorFromBool(Bool=$User.IsFrontEnd, True Color=<rgba 1 1 1 0.5>, False Color=<rgba 1 1 1 1>)
    Lifetime=&Multiply_Float(
     A=&SelectFloatFromArray(
      Direct Array Index=&SpawnGroup()
      Float Selection Array=DataInterface() {propertyValues {FloatData [0,0.5,0.5,0.4,0.4,0.35,0.35]}}
     )
     B=&MakeCustomFloatFromBool(Bool=$User.IsLargeBody, True Float=1.1, False Float=1)
    )
    Ribbon Width=&SelectFloatFromArray(
     Direct Array Index=&SpawnGroup()
     Float Selection Array=DataInterface() {propertyValues {FloatData [50,35,35,50,50,40,40]}}
    )
    Ribbon Width Mode=::NewEnumerator1
    Sprite Size Mode=::NewEnumerator3
    UsePositionOffset=false
    Write Material Random=true
   )
   /SetVariables(Particles.RibbonID=/Niagara/DynamicInputs/NiagaraID/MakeNiagaraIDFromInt(Index:NiagaraInt32=&SpawnGroup(), Acquire Tag:NiagaraInt32=0)) {isSetParameters=true}
   &AddVelocity(Rotation Coordinate Space=::World, Velocity=<0 0 200>)
   &SkeletalMeshLocation(
    Bone / Socket Index=&SpawnGroup()
    Bone Sampling Mode=::NewEnumerator4
    MaskSpawnGroup=false
    Orientation Sampling=::NewEnumerator1
    Sampled Position Offset=&SelectVectorFromArray(Direct Array Index=&SpawnGroup(), Vector Selection Array=$System.BoneOffset)
    Skeletal Mesh=$System.Skeletal Mesh
   ) {module=SkeletalMeshLocation001}
   &SkeletalMeshLocation(
    Bone / Socket Index=&SelectIntFromArray(Direct Array Index=&SpawnGroup(), Int Selection Array=$System.CompanionBone)
    Bone Sampling Mode=::NewEnumerator4
    MaskSpawnGroup=false
    Orientation Sampling=::NewEnumerator1
    Position Sampling=::NewEnumerator2
    Sampled Position Offset=&SelectVectorFromArray(Direct Array Index=&SpawnGroup(), Vector Selection Array=$System.BoneOffset)
    Skeletal Mesh=$System.Skeletal Mesh
   ) {module=SkeletalMeshLocation003}
   &InheritVelocity(
    Inherited Velocity Amount Scale=&MakeCustomVectorFromBool(Bool=$User.IsFrontEnd, True Vector=<0 0 1>, False Vector=<0.5 0.5 0.1>)
    Speed Limit=true
    UseSpeedThreshold=false
    Velocity Source=&Lerp_Vector(A=$Particles.SkeletalMeshLocation001.SampledVelocity, B=$Particles.SkeletalMeshLocation003.SampledVelocity, Alpha=0.5)
   )
   /SetVariables(Particles.RibbonFacing=&MakeCustomVectorFromBool(
     Bool=&SetBoolByIntComparison(A=&SpawnGroup(), B=0, Comparison Type=::NewEnumerator4)
     True Vector=<1 0 0>
     False Vector=&Subtract_Vector(A=$Particles.SkeletalMeshLocation001.SampledPosition, B=$Particles.SkeletalMeshLocation003.SampledPosition)
    )) {isSetParameters=true}
   &CameraOffset()
  ]
  particleUpdateScript [
   &ParticleState()
   &SkeletalMeshLocation(
    Bone / Socket Index=&SpawnGroup()
    Bone Sampling Mode=::NewEnumerator4
    MaskSpawnGroup=false
    Orientation Sampling=::NewEnumerator1
    Position Sampling=::NewEnumerator2
    Sampled Position Offset=&SelectVectorFromArray(Direct Array Index=&SpawnGroup(), Vector Selection Array=$System.BoneOffset)
    Skeletal Mesh=$System.Skeletal Mesh
   )
   &SkeletalMeshLocation(
    Bone / Socket Index=&SelectIntFromArray(Direct Array Index=&SpawnGroup(), Int Selection Array=$System.CompanionBone)
    Bone Sampling Mode=::NewEnumerator4
    MaskSpawnGroup=false
    Orientation Sampling=::NewEnumerator1
    Position Sampling=::NewEnumerator2
    Sampled Position Offset=&SelectVectorFromArray(
     Direct Array Index=&SelectIntFromArray(Direct Array Index=&SpawnGroup(), Int Selection Array=$System.CompanionBone)
     Vector Selection Array=$System.BoneOffset
    )
    Skeletal Mesh=$System.Skeletal Mesh
   ) {module=SkeletalMeshLocation002}
   &Drag(Use Linear Drag=true, Use Rotational Drag=true)
   &CurlNoiseForce(Bias Noise=false, Noise Strength=200, Pan Noise=true, Pan Noise Field=<0 0 1>)
   &SolveForcesAndVelocity(Clamp Velocity=false, Limit Acceleration=false)
   &self:IsFirstParticle(
    Locked Position:NiagaraPosition=&Lerp_Vector(
     A=Linked() {linkedVariable {name=Particles.SkeletalMeshLocation.SampledPosition, type=NiagaraPosition}}
     B=$Particles.SkeletalMeshLocation002.SampledPosition
     Alpha=0.5
    )
    Set Position=true
    Spawn Group:NiagaraInt32=&SpawnGroup()
   ) {module=IsFirstParticle}
   &self:HighSpeedControl(
    Max Expected Owner Speed=500
    Max Visible Distance=100
    Negligible Owner Speed=50
    Source Position:NiagaraPosition=$Output.IsFirstParticle.LockedPosition
   ) {module=HighSpeedControl}
   &ScaleRibbonWidth(Ribbon Width Scale=&FloatFromCurve(
     FloatCurve=DataInterface() {propertyValues {Curve {keys [
         {interpMode=RCIM_Cubic, leaveTangentWeight=0.379, tangentMode=RCTM_User, time=0, value=1}
         {arriveTangent=-1.8425, tangentMode=RCTM_User, time=1, value=0}
        ]}}}
     Scale Curve=&MakeCustomFloatFromBool(Bool=$User.IsLargeBody, True Float=1.1, False Float=1)
    ))
   &Color(Color=&MakeCustomLinearColorFromBool(Bool=$User.IsLargeBody, True Color=<rgba 0 0 1 1>, False Color=<rgba 1 0 0 1>)) {enabled=false}
   &DynamicMaterialParameters(
    Index 0 Param 1=&Multiply_FloatByInt(Float=0.1667, Integer=&SpawnGroup())
    Index 0 Param 2=&Multiply_Float(A=$Engine.Time, B=1.5)
    Module.Param0WriteEnabled=true
    Module.Param1WriteEnabled=true
    Module.Param2WriteEnabled=false
    Module.Param3WriteEnabled=false
   )
   &ScaleColor(
    Scale Alpha=&FloatFromCurve(FloatCurve=DataInterface() {propertyValues {Curve {keys [
         {arriveTangent=-3.6115, interpMode=RCIM_Cubic, leaveTangent=-3.6115, tangentMode=RCTM_User, time=0.1, value=1}
         {arriveTangent=0.0655, interpMode=RCIM_Cubic, leaveTangent=0.0655, tangentMode=RCTM_User, time=0.75, value=0.2}
         {arriveTangent=-0.9846, tangentMode=RCTM_User, time=1, value=0}
        ]}}})
    ScaleA=true
    ScaleRGB=false
   ) {enabled=false}
   /SetVariables(Particles.RestrictedPosition:NiagaraPosition=&AddVectorToPosition(
     Position=&Multiply_Vector(A=Linked() {linkedVariable {name=Particles.Position, type=NiagaraPosition}}, B=<0 0 1>)
     Vector=&Multiply_Vector(A=$Output.IsFirstParticle.LockedPosition, B=<1 1 0>)
    )) {isSetParameters=true}
   /SetVariables(Particles.PositionMod:NiagaraPosition=&LerpPosition(A=$Particles.Position, B=$Particles.RestrictedPosition, Alpha=0.5)) {isSetParameters=true}
  ]
  rendererClasses [&NiagaraRibbonRendererProperties, &NiagaraSpriteRendererProperties]
  renderers [
   {rendererClass=&NiagaraRibbonRendererProperties}
   {rendererClass=&NiagaraSpriteRendererProperties, rendererIndex=1}
   {rendererClass=&NiagaraRibbonRendererProperties, rendererIndex=2}
  ]
 }
 FlamesMesh {
  emitterUpdateScript [
   &EmitterState()
   &SpawnBurst_Instantaneous(Spawn Count=8, Use Spawn Probability=false, UseLoopCountLimit=true)
  ]
  particleSpawnScript [
   &PartitionParticles(Number Of Partitions=4, Particles Per Partition=2, Partition Mode:ENiagara_PartitionMode=::NewEnumerator1)
   /SetVariables(
    Particles.SymmetricalLerp:NiagaraFloat=&SelectFloatFromArray(
     Direct Array Index=&ReturnExecIndex()
     Float Selection Array=DataInterface() {propertyValues {FloatData [0,0.3333,0,0.3333,0.6667,1,0.6667,1]}}
    )
    Particles.float:NiagaraFloat=$Particles.PartitionParticles.Normalized Index In Partition
   ) {isSetParameters=true}
   &InitializeParticle(
    Color=&MakeCustomLinearColorFromBool(
     Bool=$User.IsFrontEnd
     True Color=&LerpLinearColors(StartColor=<rgba 1 1 1 0.5>, EndColor=<rgba 0.4844 0.4844 1 0.25>, LerpFactor=$Particles.SymmetricalLerp)
     False Color=&LerpLinearColors(StartColor=<rgba 1 1 1 1>, EndColor=<rgba 0.4844 0.4844 1 0.25>, LerpFactor=$Particles.SymmetricalLerp)
    )
    Lifetime=0.1
    Material Random:NiagaraFloat=&RandomRangeFloat(Minimum=0, Maximum=1)
    Mesh Scale=&MakeVector(
     X=1
     Y=1
     Z=&Multiply_Float(
      A=&Lerp_Float(A=1.35, B=0.55, Alpha=$Particles.SymmetricalLerp)
      B=&MakeCustomFloatFromBool(Bool=$User.IsLargeBody, True Float=1.25, False Float=1)
     )
    )
    Mesh Scale Mode=::NewEnumerator4
    Sprite Size Mode=::NewEnumerator3
    UsePositionOffset=false
    Write Material Random=true
   )
   &InitialMeshOrientation(Mesh Orientation Mode=::NewEnumerator1, Use Rotation Vector=true)
   &DynamicMaterialParameters(
    Index 0 Param 1=&RandomRangeFloat(Minimum=0, Maximum=10)
    Index 0 Param 2=25
    Index 0 Param 3=20
    Index 0 Param 4=-50
    Module.Param0WriteEnabled=true
    Module.Param1WriteEnabled=true
    Module.Param2WriteEnabled=true
    Module.Param3WriteEnabled=true
   )
  ]
  particleUpdateScript [
   &ParticleState(Kill Particles When Lifetime Has Elapsed=false, Let Infinitely Lived Particles Die When Emitter Deactivates=true)
   &UpdateMeshOrientation(
    Facing Direction=/Game/Effects/Niagara/DynamicInputs/Engine/FN_GetSystemAxis(Axis:ENiagaraOrientationAxis=::XAxis, Flip Axis:NiagaraBool=false)
    Facing Interpolation=false
    Orientation Method=::NewEnumerator1
    Override Quat=false
    TransformFacing=false
    TransformUp=false
   ) {enabled=false, module=UpdateMeshOrientation001}
   &SkeletalMeshLocation(
    Bone / Socket Index=&SelectIntFromArray(Direct Array Index=$Particles.PartitionParticles.Partition, Int Selection Array=DataInterface() {propertyValues {IntData [7,8,1,2]}})
    Bone Sampling Mode=::NewEnumerator4
    MaskSpawnGroup=false
    Orientation Sampling=::NewEnumerator1
    Sampled Position Offset=&Lerp_Vector(
     A=&SelectVectorFromArray(
      Direct Array Index=$Particles.PartitionParticles.Partition
      Vector Selection Array=DataInterface() {propertyValues {FloatData [<2 0 0>, <-2 0 0>, <10 0 0>, <-10 0 0>]}}
     )
     B=&SelectVectorFromArray(
      Direct Array Index=$Particles.PartitionParticles.Partition
      Vector Selection Array=DataInterface() {propertyValues {FloatData [<10 0 0>, <-10 0 0>, <25 0 0>, <-25 0 0>]}}
     )
     Alpha=&Multiply_Float(A=$Particles.PartitionParticles.Normalized Index In Partition, B=2)
    )
    Skeletal Mesh=$System.Skeletal Mesh
   )
   &InterpolateOverTime(Rate Of Change=5, Target Position Value:NiagaraPosition=$Particles.Position, Variable Type=::NewEnumerator6)
   /SetVariables(Particles.Interpolate Dynamic=&MakeFloat4FromVectorAndFloat("Vector (RGB)"=$Particles.InterpolateOverTime.InterpolateOverTime_Position.Moving Average, "Float (Alpha)"=1)) {isSetParameters=true}
  ]
  rendererClasses [&NiagaraMeshRendererProperties, &NiagaraSpriteRendererProperties]
  renderers [
   {rendererClass=&NiagaraMeshRendererProperties, rendererIndex=0}
   {rendererClass=&NiagaraSpriteRendererProperties, rendererIndex=1}
  ]
 }
 Minimal001 {
  bEnabled=false
  emitterUpdateScript [
   &EmitterState()
   &SpawnRate(SpawnRate=10, Use Spawn Probability=false)
  ]
  particleSpawnScript [
   &InitializeParticle(Lifetime=5, Sprite Size Mode=::NewEnumerator3, UsePositionOffset=false, Write Material Random=true)
   &AddVelocity(Velocity=<0 0 300>)
   &Color(Color=&ColorFromCurve(
     ColorCurve=DataInterface() {propertyValues {
       AlphaCurve {keys [<at 0 1 RCIM_Cubic>]}
       BlueCurve {keys [<at 0 0 RCIM_Cubic>,<at 0.5 0>,<at 1 1 RCIM_Cubic>]}
       GreenCurve {keys [<at 0 0 RCIM_Cubic>,<at 0.5 1>,<at 1 0 RCIM_Cubic>]}
       RedCurve {keys [<at 0 1 RCIM_Cubic>,<at 0.5 0>,<at 1 0 RCIM_Cubic>]}
      }}
     CurveIndex=&FN_LOD_Distance_Remap(Start Distance=300, End Distance=2500, Minimum Output=0, Maximum Output=1, Clamp Output=true)
    ))
  ]
  particleUpdateScript [
   &SolveForcesAndVelocity(Clamp Velocity=false, Limit Acceleration=false)
   &ParticleState()
  ]
  rendererClasses [&NiagaraSpriteRendererProperties]
  renderers [{rendererClass=&NiagaraSpriteRendererProperties}]
 }
 NE_FN_Sparks {
  simTarget=GPUComputeSim
  emitterUpdateScript [
   &EmitterState()
   &SpawnRate(
    SpawnRate=&FloatFromCurve(
     FloatCurve=DataInterface() {propertyValues {Curve {keys [
         {arriveTangent=-0.0017, interpMode=RCIM_Cubic, leaveTangent=-0.0017, tangentMode=RCTM_User, time=300, value=10}
         {arriveTangent=-0.0055, interpMode=RCIM_Cubic, leaveTangent=-0.0055, tangentMode=RCTM_User, time=1412.7, value=6}
         {arriveTangent=-0.0006, tangentMode=RCTM_User, time=2500, value=1}
        ]}}}
     CurveIndex=$Engine.Owner.LODDistance
     Scale Curve=15
    )
    Use Spawn Probability=false
   )
   &SpawnRate(SpawnGroup=1, SpawnRate=25, Use Spawn Probability=false) {module=SpawnRate001}
  ]
  particleSpawnScript [
   /SetVariables(
    Particles.PickBone:NiagaraInt32=&RandomRangeInt(Minimum=1, Maximum=4)
    Particles.SparkIntensity:NiagaraFloat=&FloatFromCurve(
     FloatCurve=DataInterface() {propertyValues {Curve {keys [
         <at 0 1 RCIM_Cubic>
         {arriveTangent=-2.5317, interpMode=RCIM_Cubic, leaveTangent=-2.5317, tangentMode=RCTM_User, time=0.6, value=0.5}
         <at 1 0 RCIM_Cubic>
        ]}}}
     CurveIndex=&RandomRangeFloat(Minimum=0, Maximum=1)
    )
    Particles.VisibilityTag=&SpawnGroup()
   ) {isSetParameters=true}
   &InitializeParticle(
    Color Channel Mode:ENiagaraChannelCorrelation=::NewEnumerator0
    Color Maximum=<rgba 0.3 3 1.8 1>
    Color Minimum=<rgba 0.08 1 0.64 0.5>
    Color Mode=::NewEnumerator2
    Lifetime=&Lerp_Float(
     A=&RandomRangeFloat(Minimum=0.3, Maximum=0.4)
     B=&RandomRangeFloat(Minimum=0.5, Maximum=0.8)
     Alpha=$Particles.SparkIntensity
    )
    Mass=&Lerp_Float(A=0.3, B=1, Alpha=$Particles.SparkIntensity)
    Mass Mode=::NewEnumerator1
    Sprite Size Mode=::NewEnumerator3
    Uniform Sprite Size=&MaskFloatBySpawnGroup(
     Masked Float=&Lerp_Float(A=0.45, B=0.9, Alpha=$Particles.SparkIntensity)
     Passthrough Float=&RandomRangeFloat(Minimum=7, Maximum=15)
    )
    UsePositionOffset=false
    Write Material Random=true
   )
   &SkeletalMeshLocation(
    Bone / Socket Index=$Particles.PickBone
    Bone Sampling Mode=::NewEnumerator4
    MaskSpawnGroup=false
    Orientation Sampling=::NewEnumerator1
    Position Sampling=::NewEnumerator2
    Sampled Position Offset=&MakeCustomVectorFromBool(
     Bool=$User.IsLargeBody
     True Vector=&SelectVectorFromArray(Direct Array Index=$Particles.PickBone, Vector Selection Array=$System.BoneOffset)
     False Vector=<0 0 0>
    )
    Skeletal Mesh=$System.Skeletal Mesh
   ) {module=SkeletalMeshLocation001}
   &SkeletalMeshLocation(
    Bone / Socket Index=&SelectIntFromArray(Direct Array Index=$Particles.PickBone, Int Selection Array=$System.CompanionBone)
    Bone Sampling Mode=::NewEnumerator4
    MaskSpawnGroup=false
    Orientation Sampling=::NewEnumerator1
    Position Sampling=::NewEnumerator2
    Sampled Position Offset=&MakeCustomVectorFromBool(
     Bool=$User.IsLargeBody
     True Vector=&SelectVectorFromArray(
      Direct Array Index=&SelectIntFromArray(Direct Array Index=$Particles.PickBone, Int Selection Array=$System.CompanionBone)
      Vector Selection Array=$System.BoneOffset
     )
     False Vector=<0 0 0>
    )
    Skeletal Mesh=$System.Skeletal Mesh
   )
   &ShapeLocation(
    MaskSpawnGroup=false
    Override Seed=false
    Shape Origin=&LerpPosition(
     A=$Particles.SkeletalMeshLocation001.SampledPosition
     B=$Particles.SkeletalMeshLocation.SampledPosition
     Alpha=&RandomRangeFloat(Minimum=0, Maximum=1)
    )
    Sphere Radius=1
    UseOwnerScale=true
   )
   &AddVelocityInCone(
    Cone Angle=30
    Cone Axis=<0 0 1>
    Module.Use Velocity Falloff On Cone Axis=false
    Velocity Distribution Along Cone Axis=0.8
    Velocity Strength=&MakeCustomFloatFromBool(
     Bool=$User.IsLargeBody
     True Float=&RandomRangeFloat(Minimum=60, Maximum=100)
     False Float=&RandomRangeFloat(Minimum=30, Maximum=60)
    )
   )
   &CameraOffset(Camera Offset Amount=-20)
  ]
  particleUpdateScript [
   &ParticleState()
   &SpriteRotationRate(Rotation Rate=&RandomRangeFloat(Minimum=-200, Maximum=200)) {enabled=false}
   &GravityForce(Gravity=&MakeVector(
     X=0
     Y=0
     Z=&FloatFromCurve(FloatCurve=DataInterface() {propertyValues {Curve {keys [<at 0 10>,<at 1 90>]}}})
    ))
   &Drag(Use Linear Drag=true, Use Rotational Drag=false)
   &VectorNoiseForce(
    Force Amount=&FloatFromCurve(
     FloatCurve=DataInterface() {propertyValues {Curve {keys [
         {arriveTangent=-9.4515, interpMode=RCIM_Cubic, leaveTangent=-4.1687, tangentMode=RCTM_User, time=0, value=1}
         {arriveTangent=-3.4372, interpMode=RCIM_Cubic, leaveTangent=-3.4372, tangentMode=RCTM_User, time=0.6, value=0}
        ]}}}
     Scale Curve=500
    )
    Module.Use Random Noise Time Interval=false
   )
   &SolveForcesAndVelocity(Clamp Velocity=false, Limit Acceleration=false)
   &ScaleColor(
    Linear Color Curve=DataInterface() {propertyValues {
      AlphaCurve {keys [<at 0 1>,<at 1 1>]}
      BlueCurve {keys [
        {arriveTangent=-36.8273, interpMode=RCIM_Cubic, leaveTangent=-36.8273, tangentMode=RCTM_User, time=0, value=5}
        {arriveTangent=-1.452, interpMode=RCIM_Cubic, leaveTangent=-1.452, tangentMode=RCTM_User, time=0.163, value=1}
        <at 1 0.25>
       ]}
      GreenCurve {keys [
        {arriveTangent=-36.8273, interpMode=RCIM_Cubic, leaveTangent=-36.8273, tangentMode=RCTM_User, time=0, value=5}
        {arriveTangent=-1.4408, interpMode=RCIM_Cubic, leaveTangent=-1.4408, tangentMode=RCTM_User, time=0.163, value=1}
        <at 1 0.093>
       ]}
      RedCurve {keys [
        {arriveTangent=-36.8273, interpMode=RCIM_Cubic, leaveTangent=-36.8273, tangentMode=RCTM_User, time=0, value=5}
        {arriveTangent=-0.8332, interpMode=RCIM_Cubic, leaveTangent=-0.8332, tangentMode=RCTM_User, time=0.163, value=1}
        <at 1 0.0755>
       ]}
     }}
    Scale Mode=::NewEnumerator2
   )
   &ScaleSpriteSize(Scale Factor=&Vector2DFromFloat(Value=&FloatFromCurve(FloatCurve=DataInterface() {propertyValues {Curve {keys [
          {arriveTangent=-0.8333, interpMode=RCIM_Cubic, time=0, value=0}
          {arriveTangentWeight=0.1795, interpMode=RCIM_Cubic, leaveTangentWeight=0.1795, tangentMode=RCTM_User, time=0.2, value=1}
          {arriveTangent=-0.625, interpMode=RCIM_Cubic, leaveTangent=-0.625, time=0.4, value=1}
          {arriveTangent=-2.2814, interpMode=RCIM_Cubic, leaveTangent=-2.2814, tangentMode=RCTM_User, time=1, value=0.5}
         ]}}})))
   &ScaleSpriteSizeBySpeed(
    Max Scale Factor=&MaskVector2dBySpawnGroup(Masked Vector2D=<1.25 8>, Passthrough Vector2D=<1 1>, Spawn Group Mask=0)
    Sample Scale Factor By Curve=true
    Scale Factor Curve=DataInterface() {propertyValues {Curve {keys [
        {arriveTangent=0.2757, time=0, value=0}
        {arriveTangent=0.3844, interpMode=RCIM_Cubic, leaveTangent=0.3844, tangentMode=RCTM_User, time=0.3, value=0.15}
        {arriveTangent=2.2765, interpMode=RCIM_Cubic, leaveTangent=2.2765, tangentMode=RCTM_User, time=1, value=1}
       ]}}}
    Velocity Threshold:NiagaraFloat=&FloatFromCurve(FloatCurve=DataInterface() {propertyValues {Curve {keys [<at 0.25 30>,<at 1 100>]}}})
   )
   /SetVariables(Particles.StarColor:LinearColor=&LerpLinearColors(StartColor=<rgba 3 3 3 1>, EndColor=$Particles.Initial.Color, LerpFactor=$Particles.NormalizedAge)) {isSetParameters=true}
   /SetVariables(Particles.DistScale:Vector2f=&Multiply_Vector2(
     A=$Particles.SpriteSize
     B=&Vector2DFromFloat(Value=&FloatFromCurve(
       FloatCurve=DataInterface() {propertyValues {Curve {keys [
           {arriveTangent=0.0008, interpMode=RCIM_Cubic, leaveTangent=0.0008, leaveTangentWeight=548.2426, tangentMode=RCTM_User, time=300, value=1}
           {
            arriveTangent=0.0027
            arriveTangentWeight=548.25
            interpMode=RCIM_Cubic
            leaveTangent=0.0027
            leaveTangentWeight=185.1056
            tangentMode=RCTM_User
            time=1944.7
            value=3
           }
           {arriveTangent=0.0073, arriveTangentWeight=185.1033, interpMode=RCIM_Cubic, leaveTangent=0.0073, tangentMode=RCTM_User, time=2500, value=6}
          ]}}}
       CurveIndex=$Engine.Owner.LODDistance
       Scale Curve=&RandomRangeFloat(Minimum=0.5, Maximum=1)
      ))
    )) {isSetParameters=true}
  ]
  rendererClasses [&NiagaraSpriteRendererProperties]
  renderers [
   {rendererClass=&NiagaraSpriteRendererProperties}
   {rendererClass=&NiagaraSpriteRendererProperties, rendererIndex=1}
  ]
 }
}
summary {
 systemName=NS_Aura_JadeTowel
 emitters [
  {emitterName=BodyGlow, rendererClasses [&NiagaraSpriteRendererProperties]}
  {bEnabled=false, emitterName=FN_Debug, rendererClasses [&NiagaraSpriteRendererProperties]}
  {emitterName=Flames, rendererClasses [&NiagaraRibbonRendererProperties, &NiagaraSpriteRendererProperties]}
  {emitterName=NE_FN_Sparks, simTarget=GPUComputeSim, rendererClasses [&NiagaraSpriteRendererProperties]}
  {emitterName=FlamesMesh, rendererClasses [&NiagaraMeshRendererProperties, &NiagaraSpriteRendererProperties]}
  {bEnabled=false, emitterName=Minimal001, rendererClasses [&NiagaraSpriteRendererProperties]}
 ]
 userVariables [
  User.IsFrontEnd {type=NiagaraBool}
  User.IsLargeBody {type=NiagaraBool}
 ]
}
